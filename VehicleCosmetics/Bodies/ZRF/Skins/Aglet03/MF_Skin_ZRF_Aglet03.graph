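# Material function for the Aglet03 skin on the ZRF vehicle body, overriding base color and roughness from the shared body base.
# Panning dot patterns for hood and sides are brightened by two rotated, scrolling noise layers.
# Masks, including channels of a packed RGBA texture, layer in a dark gradient with a gold distance-based shape and flat grey, around two rotated logo stamps.
# Roughness derives inversely from the final color's clamped luminance.
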MaterialFunction /VehicleCosmetics/Bodies/ZRF/Skins/Aglet03/MF_Skin_ZRF_Aglet03
graph {nodes [
  n1 = Constant3Vector() {constant=<rgba 0.4076 0.4149 0.4123 1>}
  n2 = Panner(Coordinate=NamedRerouteUsage()) {speedY=0.008}
  n3 = TextureSample(UVs=Multiply(A=@n2) {constB=32}) {desc=Dots - Hood, texture=/VehicleCosmetics/SharedTextures/T_Stripes_Dots_RGB}
  n4 = NamedRerouteUsage()
  n5 = TextureSample(UVs=Multiply(A=@n2) {constB=60}) {desc=Dots - Sides, texture=&T_Stripes_Dots_RGB}
  n6 = NamedRerouteUsage()
  n7 = LinearInterpolate(
   A=LinearInterpolate(A=Constant3Vector() {constant=<rgba 1 1 1 1>}, B=@n1, Alpha=Multiply(A=@n3.G, B=@n4))
   B=LinearInterpolate(A=@n1, B=Constant3Vector() {constant=<rgba 0.3155 0.3212 0.3191 1>}, Alpha=Multiply(A=@n5.G, B=@n6))
   Alpha=NamedRerouteUsage()
  )
  n8 = NamedRerouteUsage()
  n9 = Multiply(A=Constant2Vector() {g=0.2, r=0.01}, B=@n8)
  n10 = TextureSample(UVs=MaterialFunctionCall(UVs=Panner(Coordinate=@n9) {constCoordinate=1, speedY=0.03}, "Rotation Angle (0-1)"=Constant() {r=0.3}) {materialFunction=&CustomRotator}.Rotated Values) {samplerType=SAMPLERTYPE_Masks, texture=/VehicleCosmetics/SharedTextures/T_Universal_Noise_Heightmap_30_Blurred}
  n11 = Multiply(A=@n8, B=Constant2Vector() {g=0.1, r=0.01})
  n12 = TextureSample(UVs=MaterialFunctionCall(UVs=Panner(Coordinate=@n11) {constCoordinate=1, speedY=0.015}, "Rotation Angle (0-1)"=Constant() {r=0.3}) {materialFunction=&CustomRotator}.Rotated Values) {samplerType=SAMPLERTYPE_Masks, texture=&T_Universal_Noise_Heightmap_30_Blurred}
  n13 = ConstantBiasScale(None=Add(A=ConstantBiasScale(None=@n10.R) {bias=-0.1, scale=0.2}, B=ConstantBiasScale(None=@n12.R) {bias=-0.1})) {bias=-0.4, scale=0.7}
  n14 = Constant3Vector() {constant=<rgba 0.117 0.117 0.1144 0>}
  n15 = Multiply(A=@n14) {constB=3}
  n16 = TextureCoordinate()
  n17 = ComponentMask(None=@n16) {desc="Gradient lighten to\r\nmatch ref logo", r=true}
  n18 = Saturate(None=Abs(None=ConstantBiasScale(None=@n17) {bias=-0.38, scale=35}))
  n19 = Distance(A=ConstantBiasScale(None=@n16) {bias=-0.5, desc=Adjust Mid-Point, scale=2}, B=Constant2Vector() {g=0.47, r=-0.24}) {desc="Get distance from specificed\r\ncoordinate origin point."}
  n20 = LinearInterpolate(
   A=LinearInterpolate(A=@n7, B=Multiply(A=@n7) {constB=10}, Alpha=Saturate(None=@n13))
   B=LinearInterpolate(
    A=LinearInterpolate(A=@n15, B=@n14, Alpha=@n18)
    B=Constant3Vector() {constant=<rgba 0.6445 0.3185 0.0452 1>}
    Alpha=SmoothStep(Value=@n19) {constMax=0.1, constMin=0.1, constValue=0.1}
   )
   Alpha=NamedRerouteUsage()
  )
  n21 = TextureCoordinate() {uTiling=12, vTiling=12}
  n22 = Constant2Vector() {g=-4.5, r=-7.3}
  n23 = TextureSample(UVs=MaterialFunctionCall(UVs=Add(A=@n21, B=@n22), "Rotation Angle (0-1)"=Constant() {r=0.25}) {materialFunction=&CustomRotator}.Rotated Values) {texture=&/T_Aglet03_Logo}
  n24 = Constant2Vector() {g=-4.5, r=-0.8}
  n25 = TextureSample(UVs=MaterialFunctionCall(UVs=Add(A=@n21, B=@n24), "Rotation Angle (0-1)"=Constant() {r=0.75}) {materialFunction=&CustomRotator}.Rotated Values) {texture=&T_Aglet03_Logo}
  n26 = Multiply(A=@n23.RGB, B=@n25.RGB) {desc=Color}
  n27 = Add(A=@n23.A, B=@n25.A) {desc=Alpha}
  n28 = Saturate(None=LinearInterpolate(A=LinearInterpolate(A=@n20, B=@n26, Alpha=@n27), B=Constant3Vector() {constant=<rgba 0.0976 0.0976 0.0976 1>}, Alpha=NamedRerouteUsage()))
  n29 = Multiply(A=Desaturation(None=@n28)) {constB=0.8}
  n30 = OneMinus(None=Multiply(A=Clamp(None=@n29) {maxDefault=0.9, minDefault=0.7}) {constB=0.5})
  FunctionOutput(None=SetMaterialAttributes!(
    MaterialAttributes=MaterialFunctionCall() {materialFunction=/VehicleCosmetics/Materials/MaterialFunctions/MF_Vehicle_Body_Base}.Result
    BaseColor=@n28
    Roughness=Saturate(None=@n30)
   ) {attributeSetTypes [#1, #2]})
  NamedRerouteDeclaration(None=TextureCoordinate() {coordinateIndex=1}) {name=TexCoord1_Out, nodeColor=<rgba 1 0 0.8437 1>}
  n31 = TextureSample() {samplerType=SAMPLERTYPE_Masks, texture=&/T_Aglet03_RGB}
  NamedRerouteDeclaration(None=@n31.R) {name=R_Out, nodeColor=<rgba 0 1 0.2969 1>}
  NamedRerouteDeclaration(None=@n31.G) {name=G_Out, nodeColor=<rgba 0.125 1 0 1>}
  NamedRerouteDeclaration(None=@n31.B) {name=B_Out, nodeColor=<rgba 0 0.225 1 1>}
  NamedRerouteDeclaration(None=@n31.A) {name=A_Out, nodeColor=<rgba 0.3407 0.342 0.3282 1>}
 ]}
